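# Physics asset for the Vacation drum kit cosmetic, providing bodies and joints for its dynamic leaf and pineapple bones.
# Four three-segment capsule chains (left and right leaves, left and right cymbal leaves) hang from tiny root bodies.
# The pineapple bunch and center groups each fan four capsules out from a single root sphere.
# Narrow swing limits, with twist locked on most joints, restrict motion to a slight sway.
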PhysicsAsset /SparksCosmetics/Instrument/DrumKitHighVert/Vacation/Meshes/Vacation_DrumKit_Physics
bodies {
 dyn_leaf_cymbal_l_leaf_01 [{shapeType=Capsule, length=4.3324, radius=8, shapeName=dyn_leaf_cymbal_l_leaf_01_capsule, rotation=<rot 0 0 90>}]
 dyn_leaf_cymbal_l_leaf_02 [{shapeType=Capsule, length=19.7491, radius=8, shapeName=dyn_leaf_cymbal_l_leaf_02_capsule, rotation=<rot 0 0 90>}]
 dyn_leaf_cymbal_l_leaf_03 [{shapeType=Capsule, length=11.8243, radius=8, shapeName=dyn_leaf_cymbal_l_leaf_03_capsule, rotation=<rot 0 0 90>}]
 dyn_leaf_cymbal_l_root [{shapeType=Sphere, radius=1, shapeName=dyn_leaf_cymbal_l_root_sphere}]
 dyn_leaf_cymbal_r_leaf_01 [{shapeType=Capsule, length=4.3324, radius=8, shapeName=dyn_leaf_cymbal_l_leaf_01_capsule, rotation=<rot 0 0 90>}]
 dyn_leaf_cymbal_r_leaf_02 [{shapeType=Capsule, length=21.8714, radius=8, shapeName=dyn_leaf_cymbal_l_leaf_02_capsule, rotation=<rot 0 0 90>}]
 dyn_leaf_cymbal_r_leaf_03 [{shapeType=Capsule, length=20.8534, radius=8, shapeName=dyn_leaf_cymbal_l_leaf_03_capsule, rotation=<rot 0 0 90>}]
 dyn_leaf_cymbal_r_root [{shapeType=Sphere, radius=1, shapeName=dyn_leaf_cymbal_l_root_sphere}]
 dyn_leaf_l_01 [{shapeType=Capsule, length=5.6566, radius=6.621, shapeName=dyn_leaf_l_01_capsule, center=<9 0 0>, rotation=<rot 90 -142.9913 -46.1173>}]
 dyn_leaf_l_02 [{shapeType=Capsule, length=5.4481, radius=10.4568, shapeName=dyn_leaf_l_02_capsule, center=<7.3 0 0>, rotation=<rot 90 180 -86.7637>}]
 dyn_leaf_l_03 [{
   shapeType=Capsule
   length=7.1625
   radius=12.5084
   shapeName=dyn_leaf_l_03_capsule
   center=<7.8 0 0>
   rotation=<rot 62.8614 -179.9991 -91.9458>
  }]
 dyn_leaf_l_root [{shapeType=Capsule, length=0, radius=1, shapeName=dyn_leaf_l_root_capsule}]
 dyn_leaf_r_01 [{shapeType=Capsule, length=7.3901, radius=6.1427, shapeName=dyn_leaf_r_01_capsule, center=<8.4 0 0>, rotation=<rot 90 161.4935 57.4195>}]
 dyn_leaf_r_02 [{shapeType=Capsule, length=10.4554, radius=8.6053, shapeName=dyn_leaf_r_02_capsule, center=<6.5 0 0>, rotation=<rot 90 176.4237 -99.9409>}]
 dyn_leaf_r_03 [{shapeType=Capsule, length=5.3825, radius=10.4734, shapeName=dyn_leaf_r_03_capsule, center=<10 0 0>, rotation=<rot 90 -90 -175.3628>}]
 dyn_leaf_r_root [{shapeType=Capsule, length=0, radius=1, shapeName=dyn_leaf_r_root_capsule}]
 dyn_pineapple_bunch_01 [{
   shapeType=Capsule
   length=11.5903
   radius=5.686
   shapeName=dyn_pineapple_bunch_01_capsule
   center=<12.7107 -2.0093 0.542>
   rotation=<rot -69.8059 -176.5503 172.2444>
  }]
 dyn_pineapple_bunch_02 [{
   shapeType=Capsule
   length=13.2037
   radius=6.3344
   shapeName=dyn_pineapple_bunch_02_capsule
   center=<10.5664 3.4633 2.3747>
   rotation=<rot -4.8051 84.6999 103.6316>
  }]
 dyn_pineapple_bunch_03 [{
   shapeType=Capsule
   length=11.9983
   radius=4.4803
   shapeName=dyn_pineapple_bunch_03_capsule
   center=<14.7521 0 5.7861>
   rotation=<rot -88.518 0 0>
  }]
 dyn_pineapple_bunch_04 [{
   shapeType=Capsule
   length=10.9376
   radius=5.4175
   shapeName=dyn_pineapple_bunch_04_capsule
   center=<14.5427 -0.7104 -0.8709>
   rotation=<rot -74.9966 166.9489 -150.0395>
  }]
 dyn_pineapple_bunch_root [{shapeType=Sphere, radius=1, shapeName=dyn_pineapple_bunch_root_sphere}]
 dyn_pineapple_center_01 [{
   shapeType=Capsule
   length=8.0787
   radius=3.7776
   shapeName=dyn_pineapple_center_01_capsule
   center=<9.9338 0.8131 12.5392>
   rotation=<rot -73.2684 11.1329 -14.6765>
  }]
 dyn_pineapple_center_02 [{
   shapeType=Capsule
   length=12.3138
   radius=3.9447
   shapeName=dyn_pineapple_center_02_capsule
   center=<13.2 0 9.7>
   rotation=<rot -83.3611 -24.1356 12.3466>
  }]
 dyn_pineapple_center_03 [{
   shapeType=Capsule
   length=9.6782
   radius=3.3395
   shapeName=dyn_pineapple_center_03_capsule
   center=<6.7685 -0.7951 14.1431>
   rotation=<rot -48.4592 0 2.6599>
  }]
 dyn_pineapple_center_04 [{
   shapeType=Capsule
   length=10.9978
   radius=3.556
   shapeName=dyn_pineapple_center_04_capsule
   center=<6.1045 0 15.0061>
   rotation=<rot -51.2131 -0.0001 0.0001>
  }]
 dyn_pineapple_center_root [{shapeType=Sphere, radius=1, shapeName=dyn_pineapple_center_root_sphere}]
}
constraints [
 {
  bone1Name=dyn_pineapple_center_01
  bone2Name=dyn_pineapple_center_root
  swing1LimitDegrees=2
  swing2LimitDegrees=2
  twistLimitDegrees=45
  twistMotion=Locked
 }
 {
  bone1Name=dyn_pineapple_center_02
  bone2Name=dyn_pineapple_center_root
  swing1LimitDegrees=5
  swing2LimitDegrees=5
  twistLimitDegrees=45
  twistMotion=Locked
 }
 {
  bone1Name=dyn_pineapple_center_03
  bone2Name=dyn_pineapple_center_root
  swing1LimitDegrees=2
  swing2LimitDegrees=2
  twistLimitDegrees=45
  twistMotion=Locked
 }
 {
  bone1Name=dyn_pineapple_center_04
  bone2Name=dyn_pineapple_center_root
  swing1LimitDegrees=5
  swing2LimitDegrees=5
  twistLimitDegrees=45
  twistMotion=Locked
 }
 {bone1Name=dyn_leaf_l_01, bone2Name=dyn_leaf_l_root, swing1LimitDegrees=2, swing2LimitDegrees=2, twistLimitDegrees=1}
 {bone1Name=dyn_leaf_l_02, bone2Name=dyn_leaf_l_01, swing1LimitDegrees=10, swing2LimitDegrees=10, twistLimitDegrees=2}
 {bone1Name=dyn_leaf_l_03, bone2Name=dyn_leaf_l_02, swing1LimitDegrees=10, swing2LimitDegrees=10, twistLimitDegrees=2}
 {bone1Name=dyn_leaf_r_01, bone2Name=dyn_leaf_r_root, swing1LimitDegrees=2, swing2LimitDegrees=5, twistLimitDegrees=1, twistMotion=Locked}
 {bone1Name=dyn_leaf_r_02, bone2Name=dyn_leaf_r_01, swing1LimitDegrees=10, swing2LimitDegrees=5, twistLimitDegrees=45, twistMotion=Locked}
 {bone1Name=dyn_leaf_r_03, bone2Name=dyn_leaf_r_02, swing1LimitDegrees=10, swing2LimitDegrees=5, twistLimitDegrees=45, twistMotion=Locked}
 {bone1Name=dyn_pineapple_bunch_01, bone2Name=dyn_pineapple_bunch_root, swing1LimitDegrees=5, swing2LimitDegrees=2, twistLimitDegrees=1}
 {
  bone1Name=dyn_pineapple_bunch_02
  bone2Name=dyn_pineapple_bunch_root
  swing1LimitDegrees=5
  swing2LimitDegrees=2
  twistLimitDegrees=1
  twistMotion=Locked
 }
 {bone1Name=dyn_pineapple_bunch_03, bone2Name=dyn_pineapple_bunch_root, swing1LimitDegrees=2, swing2LimitDegrees=2, twistLimitDegrees=1}
 {bone1Name=dyn_pineapple_bunch_04, bone2Name=dyn_pineapple_bunch_root, swing1LimitDegrees=5, swing2LimitDegrees=5, twistLimitDegrees=1}
 {
  bone1Name=dyn_leaf_cymbal_l_leaf_01
  bone2Name=dyn_leaf_cymbal_l_root
  swing1LimitDegrees=2
  swing2LimitDegrees=5
  twistLimitDegrees=2
  twistMotion=Locked
 }
 {
  bone1Name=dyn_leaf_cymbal_l_leaf_02
  bone2Name=dyn_leaf_cymbal_l_leaf_01
  swing1LimitDegrees=10
  swing2LimitDegrees=5
  twistLimitDegrees=45
  twistMotion=Locked
 }
 {
  bone1Name=dyn_leaf_cymbal_l_leaf_03
  bone2Name=dyn_leaf_cymbal_l_leaf_02
  swing1LimitDegrees=5
  swing2LimitDegrees=8
  twistLimitDegrees=45
  twistMotion=Locked
 }
 {
  bone1Name=dyn_leaf_cymbal_r_leaf_01
  bone2Name=dyn_leaf_cymbal_r_root
  swing1LimitDegrees=2
  swing2LimitDegrees=5
  twistLimitDegrees=2
  twistMotion=Locked
 }
 {
  bone1Name=dyn_leaf_cymbal_r_leaf_02
  bone2Name=dyn_leaf_cymbal_r_leaf_01
  swing1LimitDegrees=10
  swing2LimitDegrees=5
  twistLimitDegrees=45
  twistMotion=Locked
 }
 {
  bone1Name=dyn_leaf_cymbal_r_leaf_03
  bone2Name=dyn_leaf_cymbal_r_leaf_02
  swing1LimitDegrees=5
  swing2LimitDegrees=8
  twistLimitDegrees=45
  twistMotion=Locked
 }
]
