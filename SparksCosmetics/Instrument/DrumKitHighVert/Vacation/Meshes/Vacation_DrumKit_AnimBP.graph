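# Animation Blueprint for the Vacation drum kit cosmetic that procedurally sways its leaves, flowers and pineapples.
# Each update sets variables to real-time sine waves with amplitudes re-randomized every tick, adding jitter.
# The AnimGraph runs the cached input pose through RigidBody physics in root-bone space with overridden gravity.
# A chain of additive, parent-space Transform Bone nodes then applies those variables as rotations, reusing waves across bones, before output.
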AnimBlueprint /SparksCosmetics/Instrument/DrumKitHighVert/Vacation/Meshes/Vacation_DrumKit_AnimBP
target_skeleton=&/Vacation_DrumKit_Skeleton
graphs [
 AnimGraph {nodes [
   Apply Additive(Base=Component To Local()) {alphaBoolBlend {blendOption=Linear}}
   Component To Local()
   n1 = RigidBody(ComponentPose=Local To Component(LocalPose=Use cached pose 'cached input pose'())) {
    bOverrideWorldGravity=true
    overridePhysicsAsset=&/Vacation_DrumKit_Physics
    simulationSpace=BaseBoneSpace
    alphaBoolBlend {blendOption=Linear}
    baseBoneRef {boneName=root}
    componentAppliedLinearAccClamp=<1000 1000 1000>
    componentLinearAccScale=<0.2 0.2 0.2>
    componentLinearVelScale=<0.1 0.1 0.1>
    overrideWorldGravity {z=-980}
    simSpaceSettings {worldAlpha=1, externalLinearVelocity {y=500, z=500}}
   }
   n2 = "Transform (Modify) Bone - Bone: dyn_leaf_l_root"!(
    ComponentPose=@n1
    bAlphaBoolEnabled:bool=True
    Translation="0,0,0"
    Rotation_Roll:float=0
    Rotation_Pitch:float=Get wave_01() -> (wave_01:double)
    Rotation_Yaw:float=0
    Scale="1,1,1"
   ) -> (Pose) {
    alphaInputType=Bool
    rotationMode=BMM_Additive
    rotationSpace=BCS_ParentBoneSpace
    alphaBoolBlend {blendOption=Linear}
    boneToModify {boneName=dyn_leaf_l_root}
   }
   n3 = Get Shake 3() -> (Shake 3:double)
   n4 = "Transform (Modify) Bone - Bone: dyn_leaf_l_02"!(
    ComponentPose=@n2
    bAlphaBoolEnabled:bool=True
    Translation="0,0,0"
    Rotation_Roll:float=@n3
    Rotation_Pitch:float=0
    Rotation_Yaw:float=@n3
    Scale="1,1,1"
   ) -> (Pose) {
    alphaInputType=Bool
    rotationMode=BMM_Additive
    rotationSpace=BCS_ParentBoneSpace
    alphaBoolBlend {blendOption=Linear}
    boneToModify {boneName=dyn_leaf_l_02}
   }
   n5 = "Transform (Modify) Bone - Bone: dyn_leaf_r_root"!(
    ComponentPose=@n4
    bAlphaBoolEnabled:bool=True
    Translation="0,0,0"
    Rotation_Roll:float=0
    Rotation_Pitch:float=Get wave_02() -> (wave_02:double)
    Rotation_Yaw:float=0
    Scale="1,1,1"
   ) -> (Pose) {
    alphaInputType=Bool
    rotationMode=BMM_Additive
    rotationSpace=BCS_ParentBoneSpace
    alphaBoolBlend {blendOption=Linear}
    boneToModify {boneName=dyn_leaf_r_root}
   }
   n6 = "Transform (Modify) Bone - Bone: dyn_leaf_r_02"!(
    ComponentPose=@n5
    bAlphaBoolEnabled:bool=True
    Translation="0,0,0"
    Rotation_Roll:float=Get shake 1() -> (shake 1:double)
    Rotation_Pitch:float=0
    Rotation_Yaw:float=0
    Scale="1,1,1"
   ) -> (Pose) {
    alphaInputType=Bool
    rotationMode=BMM_Additive
    rotationSpace=BCS_ParentBoneSpace
    alphaBoolBlend {blendOption=Linear}
    boneToModify {boneName=dyn_leaf_r_02}
   }
   n7 = "Transform (Modify) Bone - Bone: dyn_leaf_r_03"!(
    ComponentPose=@n6
    bAlphaBoolEnabled:bool=True
    Translation="0,0,0"
    Rotation_Roll:float=0
    Rotation_Pitch:float=Get shake 2() -> (shake 2:double)
    Rotation_Yaw:float=0
    Scale="1,1,1"
   ) -> (Pose) {
    alphaInputType=Bool
    rotationMode=BMM_Additive
    rotationSpace=BCS_ParentBoneSpace
    alphaBoolBlend {blendOption=Linear}
    boneToModify {boneName=dyn_leaf_r_03}
   }
   n8 = "Transform (Modify) Bone - Bone: dyn_leaf_cymbal_l_root"!(
    ComponentPose=@n7
    bAlphaBoolEnabled:bool=True
    Translation="0,0,0"
    Rotation_Roll:float=0
    Rotation_Pitch:float=Get wave 3() -> (wave 3:double)
    Rotation_Yaw:float=0
    Scale="1,1,1"
   ) -> (Pose) {
    alphaInputType=Bool
    rotationMode=BMM_Additive
    rotationSpace=BCS_ParentBoneSpace
    alphaBoolBlend {blendOption=Linear}
    boneToModify {boneName=dyn_leaf_cymbal_l_root}
   }
   n9 = "Transform (Modify) Bone - Bone: dyn_leaf_cymbal_l_leaf_03"!(
    ComponentPose=@n8
    bAlphaBoolEnabled:bool=True
    Translation="0,0,0"
    Rotation_Roll:float=Get shake 1()
    Rotation_Pitch:float=0
    Rotation_Yaw:float=0
    Scale="1,1,1"
   ) -> (Pose) {
    alphaInputType=Bool
    rotationMode=BMM_Additive
    rotationSpace=BCS_ParentBoneSpace
    alphaBoolBlend {blendOption=Linear}
    boneToModify {boneName=dyn_leaf_cymbal_l_leaf_03}
   }
   n10 = "Transform (Modify) Bone - Bone: dyn_leaf_cymbal_r_root"!(
    ComponentPose=@n9
    bAlphaBoolEnabled:bool=True
    Translation="0,0,0"
    Rotation_Roll:float=0
    Rotation_Pitch:float=Get wave 4() -> (wave 4:double)
    Rotation_Yaw:float=0
    Scale="1,1,1"
   ) -> (Pose) {
    alphaInputType=Bool
    rotationMode=BMM_Additive
    rotationSpace=BCS_ParentBoneSpace
    alphaBoolBlend {blendOption=Linear}
    boneToModify {boneName=dyn_leaf_cymbal_r_root}
   }
   n11 = "Transform (Modify) Bone - Bone: dyn_leaf_cymbal_r_leaf_02"!(
    ComponentPose=@n10
    bAlphaBoolEnabled:bool=True
    Translation="0,0,0"
    Rotation_Roll:float=Get shake 2()
    Rotation_Pitch:float=0
    Rotation_Yaw:float=0
    Scale="1,1,1"
   ) -> (Pose) {
    alphaInputType=Bool
    rotationMode=BMM_Additive
    rotationSpace=BCS_ParentBoneSpace
    alphaBoolBlend {blendOption=Linear}
    boneToModify {boneName=dyn_leaf_cymbal_r_leaf_02}
   }
   n12 = "Transform (Modify) Bone - Bone: dyn_flower_r_01"!(
    ComponentPose=@n11
    bAlphaBoolEnabled:bool=True
    Translation="0,0,0"
    Rotation_Roll:float=0
    Rotation_Pitch:float=Get shake 1()
    Rotation_Yaw:float=Get flower_wave 1() -> (flower_wave 1:double)
    Scale="1,1,1"
   ) -> (Pose) {
    alphaInputType=Bool
    rotationMode=BMM_Additive
    rotationSpace=BCS_ParentBoneSpace
    alphaBoolBlend {blendOption=Linear}
    boneToModify {boneName=dyn_flower_r_01}
   }
   n13 = "Transform (Modify) Bone - Bone: dyn_flower_r_02"!(
    ComponentPose=@n12
    bAlphaBoolEnabled:bool=True
    Translation="0,0,0"
    Rotation_Roll:float=0
    Rotation_Pitch:float=Get shake 2()
    Rotation_Yaw:float=Get flower wave 2() -> (flower wave 2:double)
    Scale="1,1,1"
   ) -> (Pose) {
    alphaInputType=Bool
    rotationMode=BMM_Additive
    rotationSpace=BCS_ParentBoneSpace
    alphaBoolBlend {blendOption=Linear}
    boneToModify {boneName=dyn_flower_r_02}
   }
   n14 = "Transform (Modify) Bone - Bone: dyn_flower_r_03"!(
    ComponentPose=@n13
    bAlphaBoolEnabled:bool=True
    Translation="0,0,0"
    Rotation_Roll:float=0
    Rotation_Pitch:float=Get shake 1()
    Rotation_Yaw:float=Get flower wave 3() -> (flower wave 3:double)
    Scale="1,1,1"
   ) -> (Pose) {
    alphaInputType=Bool
    rotationMode=BMM_Additive
    rotationSpace=BCS_ParentBoneSpace
    alphaBoolBlend {blendOption=Linear}
    boneToModify {boneName=dyn_flower_r_03}
   }
   n15 = "Transform (Modify) Bone - Bone: dyn_flower_r_04"!(
    ComponentPose=@n14
    bAlphaBoolEnabled:bool=True
    Translation="0,0,0"
    Rotation_Roll:float=0
    Rotation_Pitch:float=Get shake 1()
    Rotation_Yaw:float=Get flower_wave 1()
    Scale="1,1,1"
   ) -> (Pose) {
    alphaInputType=Bool
    rotationMode=BMM_Additive
    rotationSpace=BCS_ParentBoneSpace
    alphaBoolBlend {blendOption=Linear}
    boneToModify {boneName=dyn_flower_r_04}
   }
   n16 = "Transform (Modify) Bone - Bone: dyn_flower_r_05"!(
    ComponentPose=@n15
    bAlphaBoolEnabled:bool=True
    Translation="0,0,0"
    Rotation_Roll:float=0
    Rotation_Pitch:float=Get shake 1()
    Rotation_Yaw:float=Get flower wave 2()
    Scale="1,1,1"
   ) -> (Pose) {
    alphaInputType=Bool
    rotationMode=BMM_Additive
    rotationSpace=BCS_ParentBoneSpace
    alphaBoolBlend {blendOption=Linear}
    boneToModify {boneName=dyn_flower_r_05}
   }
   n17 = "Transform (Modify) Bone - Bone: dyn_flower_r_06"!(
    ComponentPose=@n16
    bAlphaBoolEnabled:bool=True
    Translation="0,0,0"
    Rotation_Roll:float=0
    Rotation_Pitch:float=Get shake 1()
    Rotation_Yaw:float=Get flower wave 3()
    Scale="1,1,1"
   ) -> (Pose) {
    alphaInputType=Bool
    rotationMode=BMM_Additive
    rotationSpace=BCS_ParentBoneSpace
    alphaBoolBlend {blendOption=Linear}
    boneToModify {boneName=dyn_flower_r_06}
   }
   n18 = "Transform (Modify) Bone - Bone: dyn_flower_l_01"!(
    ComponentPose=@n17
    bAlphaBoolEnabled:bool=True
    Translation="0,0,0"
    Rotation_Roll:float=0
    Rotation_Pitch:float=Get shake 2()
    Rotation_Yaw:float=Get flower wave 2()
    Scale="1,1,1"
   ) -> (Pose) {
    alphaInputType=Bool
    rotationMode=BMM_Additive
    rotationSpace=BCS_ParentBoneSpace
    alphaBoolBlend {blendOption=Linear}
    boneToModify {boneName=dyn_flower_l_01}
   }
   n19 = "Transform (Modify) Bone - Bone: dyn_flower_l_02"!(
    ComponentPose=@n18
    bAlphaBoolEnabled:bool=True
    Translation="0,0,0"
    Rotation_Roll:float=0
    Rotation_Pitch:float=Get shake 2()
    Rotation_Yaw:float=Get flower wave 3()
    Scale="1,1,1"
   ) -> (Pose) {
    alphaInputType=Bool
    rotationMode=BMM_Additive
    rotationSpace=BCS_ParentBoneSpace
    alphaBoolBlend {blendOption=Linear}
    boneToModify {boneName=dyn_flower_l_02}
   }
   n20 = "Transform (Modify) Bone - Bone: dyn_flower_l_03"!(
    ComponentPose=@n19
    bAlphaBoolEnabled:bool=True
    Translation="0,0,0"
    Rotation_Roll:float=0
    Rotation_Pitch:float=Get shake 2()
    Rotation_Yaw:float=Get flower wave 4() -> (flower wave 4:double)
    Scale="1,1,1"
   ) -> (Pose) {
    alphaInputType=Bool
    rotationMode=BMM_Additive
    rotationSpace=BCS_ParentBoneSpace
    alphaBoolBlend {blendOption=Linear}
    boneToModify {boneName=dyn_flower_l_03}
   }
   n21 = "Transform (Modify) Bone - Bone: dyn_flower_l_04"!(
    ComponentPose=@n20
    bAlphaBoolEnabled:bool=True
    Translation="0,0,0"
    Rotation_Roll:float=0
    Rotation_Pitch:float=Get shake 2()
    Rotation_Yaw:float=Get flower_wave 1()
    Scale="1,1,1"
   ) -> (Pose) {
    alphaInputType=Bool
    rotationMode=BMM_Additive
    rotationSpace=BCS_ParentBoneSpace
    alphaBoolBlend {blendOption=Linear}
    boneToModify {boneName=dyn_flower_l_04}
   }
   n22 = "Transform (Modify) Bone - Bone: dyn_flower_l_05"!(
    ComponentPose=@n21
    bAlphaBoolEnabled:bool=True
    Translation="0,0,0"
    Rotation_Roll:float=0
    Rotation_Pitch:float=Get shake 2()
    Rotation_Yaw:float=Get flower wave 3()
    Scale="1,1,1"
   ) -> (Pose) {
    alphaInputType=Bool
    rotationMode=BMM_Additive
    rotationSpace=BCS_ParentBoneSpace
    alphaBoolBlend {blendOption=Linear}
    boneToModify {boneName=dyn_flower_l_05}
   }
   n23 = "Transform (Modify) Bone - Bone: dyn_flower_l_06"!(
    ComponentPose=@n22
    bAlphaBoolEnabled:bool=True
    Translation="0,0,0"
    Rotation_Roll:float=0
    Rotation_Pitch:float=Get shake 2()
    Rotation_Yaw:float=Get flower wave 4()
    Scale="1,1,1"
   ) -> (Pose) {
    alphaInputType=Bool
    rotationMode=BMM_Additive
    rotationSpace=BCS_ParentBoneSpace
    alphaBoolBlend {blendOption=Linear}
    boneToModify {boneName=dyn_flower_l_06}
   }
   n24 = "Transform (Modify) Bone - Bone: dyn_flower_l_07"!(
    ComponentPose=@n23
    bAlphaBoolEnabled:bool=True
    Translation="0,0,0"
    Rotation_Roll:float=0
    Rotation_Pitch:float=Get shake 2()
    Rotation_Yaw:float=Get flower wave 2()
    Scale="1,1,1"
   ) -> (Pose) {
    alphaInputType=Bool
    rotationMode=BMM_Additive
    rotationSpace=BCS_ParentBoneSpace
    alphaBoolBlend {blendOption=Linear}
    boneToModify {boneName=dyn_flower_l_07}
   }
   n25 = "Transform (Modify) Bone - Bone: dyn_pineapple_center_01"!(
    ComponentPose=@n24
    bAlphaBoolEnabled:bool=True
    Translation="0,0,0"
    Rotation_Roll:float=Get Shake 3()
    Rotation_Pitch:float=Get pineapple 1() -> (pineapple 1:double)
    Rotation_Yaw:float=0
    Scale="1,1,1"
   ) -> (Pose) {
    alphaInputType=Bool
    rotationMode=BMM_Additive
    rotationSpace=BCS_ParentBoneSpace
    alphaBoolBlend {blendOption=Linear}
    boneToModify {boneName=dyn_pineapple_center_01}
   }
   n26 = "Transform (Modify) Bone - Bone: dyn_pineapple_center_02"!(
    ComponentPose=@n25
    bAlphaBoolEnabled:bool=True
    Translation="0,0,0"
    Rotation_Roll:float=0
    Rotation_Pitch:float=0
    Rotation_Yaw:float=Get pineapple 2() -> (pineapple 2:double)
    Scale="1,1,1"
   ) -> (Pose) {
    alphaInputType=Bool
    rotationMode=BMM_Additive
    rotationSpace=BCS_ParentBoneSpace
    alphaBoolBlend {blendOption=Linear}
    boneToModify {boneName=dyn_pineapple_center_02}
   }
   n27 = "Transform (Modify) Bone - Bone: dyn_pineapple_center_03"!(
    ComponentPose=@n26
    bAlphaBoolEnabled:bool=True
    Translation="0,0,0"
    Rotation_Roll:float=Get pineapple 1()
    Rotation_Pitch:float=0
    Rotation_Yaw:float=0
    Scale="1,1,1"
   ) -> (Pose) {
    alphaInputType=Bool
    rotationMode=BMM_Additive
    rotationSpace=BCS_ParentBoneSpace
    alphaBoolBlend {blendOption=Linear}
    boneToModify {boneName=dyn_pineapple_center_03}
   }
   n28 = "Transform (Modify) Bone - Bone: dyn_pineapple_center_04"!(
    ComponentPose=@n27
    bAlphaBoolEnabled:bool=True
    Translation="0,0,0"
    Rotation_Roll:float=0
    Rotation_Pitch:float=0
    Rotation_Yaw:float=Get pineapple 2()
    Scale="1,1,1"
   ) -> (Pose) {
    alphaInputType=Bool
    rotationMode=BMM_Additive
    rotationSpace=BCS_ParentBoneSpace
    alphaBoolBlend {blendOption=Linear}
    boneToModify {boneName=dyn_pineapple_center_04}
   }
   n29 = "Transform (Modify) Bone - Bone: dyn_pineapple_bunch_01"!(
    ComponentPose=@n28
    bAlphaBoolEnabled:bool=True
    Translation="0,0,0"
    Rotation_Roll:float=Get pineapple 2()
    Rotation_Pitch:float=0
    Rotation_Yaw:float=Get shake 1()
    Scale="1,1,1"
   ) -> (Pose) {
    alphaInputType=Bool
    rotationMode=BMM_Additive
    rotationSpace=BCS_ParentBoneSpace
    alphaBoolBlend {blendOption=Linear}
    boneToModify {boneName=dyn_pineapple_bunch_01}
   }
   n30 = "Transform (Modify) Bone - Bone: dyn_pineapple_bunch_02"!(
    ComponentPose=@n29
    bAlphaBoolEnabled:bool=True
    Translation="0,0,0"
    Rotation_Roll:float=0
    Rotation_Pitch:float=0
    Rotation_Yaw:float=Get pineapple 1()
    Scale="1,1,1"
   ) -> (Pose) {
    alphaInputType=Bool
    rotationMode=BMM_Additive
    rotationSpace=BCS_ParentBoneSpace
    alphaBoolBlend {blendOption=Linear}
    boneToModify {boneName=dyn_pineapple_bunch_02}
   }
   n31 = "Transform (Modify) Bone - Bone: dyn_pineapple_bunch_03"!(
    ComponentPose=@n30
    bAlphaBoolEnabled:bool=True
    Translation="0,0,0"
    Rotation_Roll:float=Get pineapple 1()
    Rotation_Pitch:float=0
    Rotation_Yaw:float=0
    Scale="1,1,1"
   ) -> (Pose) {
    alphaInputType=Bool
    rotationMode=BMM_Additive
    rotationSpace=BCS_ParentBoneSpace
    alphaBoolBlend {blendOption=Linear}
    boneToModify {boneName=dyn_pineapple_bunch_03}
   }
   n32 = "Transform (Modify) Bone - Bone: dyn_pineapple_bunch_04"!(
    ComponentPose=@n31
    bAlphaBoolEnabled:bool=True
    Translation="0,0,0"
    Rotation_Roll:float=Get shake 1()
    Rotation_Pitch:float=0
    Rotation_Yaw:float=Get pineapple 1()
    Scale="1,1,1"
   ) -> (Pose) {
    alphaInputType=Bool
    rotationMode=BMM_Additive
    rotationSpace=BCS_ParentBoneSpace
    alphaBoolBlend {blendOption=Linear}
    boneToModify {boneName=dyn_pineapple_bunch_04}
   }
   Output Pose(Result=Component To Local(ComponentPose=@n32))
   Save cached pose 'cached input pose'(Pose=Input Pose()) {cacheName=cached input pose}
   Reroute Node(InputPin:Wildcard) -> (OutputPin:Wildcard)
  ]}
 EventGraph {nodes [
   Event BlueprintUpdateAnimation()
   > Sequence()
   n1 = GetRealTimeSeconds()
   n2 = float * float(A=@n1, B=6)
   > Set wave_01(wave_01:double=float * float(A="Sin (Radians)"(A=@n2), B=RandomFloatInRange(Min=1.8, Max=2))) -> (Output_Get:double)
   n3 = float * float(A=@n1, B=8)
   > Set wave_02(wave_02:double=float * float(A="Sin (Radians)"(A=@n3), B=RandomFloatInRange(Min=1, Max=1.5))) -> (Output_Get:double)
   n4 = float * float(A=@n1, B=4)
   > Set wave 3(wave 3:double=float * float(A="Sin (Radians)"(A=@n4), B=RandomFloatInRange(Min=1, Max=1.25))) -> (Output_Get:double)
   n5 = float * float(A=@n1, B=4)
   > Set wave 4(wave 4:double=float * float(A="Sin (Radians)"(A=@n5), B=RandomFloatInRange(Min=-1.25, Max=-1.5))) -> (Output_Get:double)
   n6 = float * float(A=@n1, B=8)
   > Set flower_wave 1(flower_wave 1:double=float * float(A="Sin (Radians)"(A=@n6), B=RandomFloatInRange(Min=1, Max=2))) -> (Output_Get:double)
   n7 = float * float(A=@n1, B=8)
   > Set flower wave 2(flower wave 2:double=float * float(A="Sin (Radians)"(A=@n7), B=RandomFloatInRange(Min=-2, Max=-3))) -> (Output_Get:double)
   n8 = float * float(A=@n1, B=10)
   > Set flower wave 3(flower wave 3:double=float * float(A="Sin (Radians)"(A=@n8), B=RandomFloatInRange(Min=1.5, Max=2.5))) -> (Output_Get:double)
   n9 = float * float(A=@n1, B=12)
   > Set flower wave 4(flower wave 4:double=float * float(A="Sin (Radians)"(A=@n9), B=RandomFloatInRange(Min=-1.5, Max=-2))) -> (Output_Get:double)
   n10 = float * float(A=@n1, B=6)
   > Set shake 1(shake 1:double=float * float(A="Sin (Radians)"(A=@n10), B=RandomFloatInRange(Min=0.75, Max=1))) -> (Output_Get:double)
   n11 = float * float(A=@n1, B=8)
   > Set shake 2(shake 2:double=float * float(A="Sin (Radians)"(A=@n11), B=RandomFloatInRange(Min=-0.75, Max=-1))) -> (Output_Get:double)
   n12 = float * float(A=@n1, B=5)
   > Set Shake 3(Shake 3:double=float * float(A="Sin (Radians)"(A=@n12), B=RandomFloatInRange(Min=1, Max=1.5))) -> (Output_Get:double)
   n13 = float * float(A=@n1, B=4)
   > Set pineapple 1(pineapple 1:double=float * float(A="Sin (Radians)"(A=@n13), B=RandomFloatInRange(Min=0.8, Max=1))) -> (Output_Get:double)
   n14 = float * float(A=@n1, B=6)
   > Set pineapple 2(pineapple 2:double=float * float(A="Sin (Radians)"(A=@n14), B=RandomFloatInRange(Min=-0.5, Max=-0.8))) -> (Output_Get:double)
   TryGetPawnOwner()
  ]}
]
variables [
 flower wave 2 {type=number}
 flower wave 3 {type=number}
 flower wave 4 {type=number}
 flower_wave 1 {type=number}
 pineapple 1 {type=number}
 pineapple 2 {type=number}
 shake 1 {type=number}
 shake 2 {type=number}
 Shake 3 {type=number}
 wave 3 {type=number}
 wave 4 {type=number}
 wave_01 {type=number}
 wave_02 {type=number}
]
